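# Translucent, unlit, two-sided particle material for an idle aurora borealis aura, flagged for Niagara and Cascade sprites.
# Panning stretched-turbulence streaks, offset by a per-particle random dynamic parameter, scale ColorA over a ColorB base for emissive.
# Opacity combines a square edge mask, particle alpha and a UV1 gradient that is warped by a panning distortion normal and boosted by the streaks.
# Depth fade softens intersections, and a view-angle term fades grazing surfaces.
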Material /Game/Effects/Fort_Effects/Materials/Aurora/M_Aura_AuroraBorealis_Idle
graph {
 properties {
  bAutomaticallySetUsageInEditor=false
  bUsedWithNiagaraSprites=true
  bUsedWithParticleSprites=true
  bUsedWithStaticMesh=false
  blendMode=BLEND_Translucent
  pixelDepthOffsetMode=PDOM_Legacy
  shadingModel=MSM_Unlit
  twoSided=true
  usedShadingModels=Unlit
 }
 nodes [
  n1 = VectorParameter() {parameterName=ColorA, defaultValue=<rgba 0 20 4 1>}
  n2 = DynamicParameter() {paramNames [rand_0-1,Param2,Param3,Param4]}
  n3 = VectorParameter() {parameterName=StreaksTileSpeed, defaultValue=<rgba 4 0.3 -0.3 -0.5>}
  n4 = Multiply(A=TextureCoordinate(), B=ComponentMask(None=@n3.RGB) {g=true, r=true})
  n5 = TextureSample(UVs=Panner(Coordinate=Add(A=@n2.rand_0-1, B=@n4), Speed=AppendVector(A=@n3.B, B=@n3.B))) {texture=/Game/Effects/Fort_Effects/Textures/Plasma/T_Turbulence_Stretched}
  MaterialFunctionCall(Emissive=Add(A=Multiply(A=@n1.RGB, B=@n5.G), B=VectorParameter() {parameterName=ColorB, defaultValue=<rgba 0 0.3404 0.6875 1>}.RGB)) {materialFunction=&MF_FX_EmissiveAdjustments}
  n6 = TextureCoordinate()
  n7 = ScalarParameter() {defaultValue=5, parameterName=EdgeMult}
  n8 = ScalarParameter() {defaultValue=3, parameterName=EdgeExp}
  n9 = Multiply(
   A=Multiply(A=MaterialFunctionCall!(UVs=@n6, Mult=@n7, Exp=@n8) {materialFunction=/Game/Effects/Fort_Effects/MaterialFunctions/MF_SquareEdgeMask}.SquareMask)
   B=ParticleColor().A
  )
  n10 = TextureCoordinate() {coordinateIndex=1}
  n11 = ComponentMask(None=@n10) {r=true}
  n12 = OneMinus(None=ComponentMask(None=@n10) {g=true})
  n13 = ScalarParameter() {defaultValue=0.4, parameterName=DistAmnt}
  n14 = VectorParameter() {parameterName=DistTileSpeed, defaultValue=<rgba 1 0.25 0.1 -0.05>}
  n15 = Add(A=@n2.rand_0-1, B=Multiply(A=TextureCoordinate(), B=ComponentMask(None=@n14.RGB) {g=true, r=true}))
  n16 = AppendVector(A=@n14.B, B=@n14.B)
  n17 = ComponentMask(None=TextureSample(UVs=Panner(Coordinate=@n15, Speed=@n16)) {samplerType=SAMPLERTYPE_Normal, texture=&T_Weap_Distort_N_02}.RGB) {g=true, r=true}
  n18 = TextureSample(UVs=Add(A=AppendVector(A=@n11, B=@n12), B=Multiply(A=@n13, B=@n17))) {texture=/Game/Effects/Fort_Effects/Textures/Grad/gradient_healed}
  n19 = Saturate(None=Add(A=@n18.G, B=Multiply(A=@n5.G, B=@n18.G)))
  n20 = DotProduct(A=CameraVectorWS(), B=Multiply(A=VertexNormalWS(), B=TwoSidedSign()))
  n21 = ScalarParameter() {defaultValue=3, parameterName=FresnelExp}
  Multiply(
   A=DepthFade(Opacity=Multiply(A=@n9, B=@n19), FadeDistance=ScalarParameter() {defaultValue=20, parameterName=DepthDist})
   B=Saturate(None=Power(Base=@n20, Exp=@n21))
  )
 ]
}
