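# Translucent, unlit, two-sided material for aurora borealis aura trails on beam-trail particles.
# Two world-aligned plasma noises, a panning mask and scrolling background noise blend two colors that drift on a slow sine, tinted by particle color.
# Opacity combines that noise with UV edge gradients, the mask, camera and depth fades, and particle alpha.
# A separate distorted-streak branch has its own emissive and depth-fade endpoints.
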Material /Game/Effects/Fort_Effects/Materials/Aurora/M_Aura_AuroraBorealis_Trail
graph {
 properties {
  bUsedWithBeamTrails=true
  blendMode=BLEND_Translucent
  numCustomizedUVs=4
  pixelDepthOffsetMode=PDOM_Legacy
  shadingModel=MSM_Unlit
  twoSided=true
  usedShadingModels=Unlit
 }
 nodes [
  n1 = TextureObject() {texture=/Game/Effects/Fort_Effects/Textures/Plasma/T_PlasmaNoise_A_01}
  n2 = ScalarParameter() {defaultValue=1.8, parameterName=Noise A U}
  n3 = ScalarParameter() {defaultValue=0.475, parameterName=Noise A V}
  n4 = Multiply(A=AppendVector(A=AppendVector(A=@n2, B=@n3), B=Constant() {r=1})) {constB=2048}
  n5 = TextureObject() {texture=&T_PlasmaNoise_A_01}
  n6 = ScalarParameter() {defaultValue=1.2, parameterName=Noise B U}
  n7 = ScalarParameter() {defaultValue=0.941, parameterName=Noise B V}
  n8 = Multiply(A=AppendVector(A=AppendVector(A=@n6, B=@n7), B=Constant() {r=1})) {constB=2048}
  n9 = Add(
   A=ComponentMask(None=MaterialFunctionCall(TextureObject=@n1, TextureSize=@n4) {materialFunction=&WorldAlignedTexture}.XYZ Texture) {r=true}
   B=ComponentMask(None=MaterialFunctionCall(TextureObject=@n5, TextureSize=@n8) {materialFunction=&WorldAlignedTexture}.XYZ Texture) {b=true}
  )
  Clamp(None=@n9)
  n10 = TextureCoordinate()
  n11 = ComponentMask(None=@n10) {r=true}
  n12 = OneMinus(None=ComponentMask(None=@n10) {g=true})
  n13 = ScalarParameter() {defaultValue=0.4, parameterName=DistAmnt}
  n14 = DynamicParameter() {paramNames [rand_0-1,Param2,Param3,Param4]}
  n15 = VectorParameter() {parameterName=DistTileSpeed, defaultValue=<rgba 1 0.25 1 -0.05>}
  n16 = Add(A=@n14.rand_0-1, B=Multiply(A=TextureCoordinate() {coordinateIndex=1}, B=ComponentMask(None=@n15.RGB) {g=true, r=true}))
  n17 = AppendVector(A=@n15.B, B=@n15.B)
  n18 = ComponentMask(None=TextureSample(UVs=Panner(Coordinate=@n16, Speed=@n17)) {samplerType=SAMPLERTYPE_Normal, texture=&T_Weap_Distort_N_02}.RGB) {g=true, r=true}
  n19 = TextureSample(UVs=Add(A=AppendVector(A=@n11, B=@n12), B=Multiply(A=@n13, B=@n18))) {texture=/Game/Effects/Fort_Effects/Textures/Grad/gradient_healed}
  n20 = VectorParameter() {parameterName=StreaksTileSpeed, defaultValue=<rgba 8 0.3 0.3 -0.5>}
  n21 = Multiply(A=TextureCoordinate() {coordinateIndex=1}, B=ComponentMask(None=@n20.RGB) {g=true, r=true})
  n22 = TextureSample(UVs=Panner(Coordinate=Add(A=@n14.rand_0-1, B=@n21), Speed=AppendVector(A=@n20.B, B=@n20.B))) {texture=/Game/Effects/Fort_Effects/Textures/Plasma/T_Turbulence_Stretched}
  n23 = Add(A=@n19.G, B=Multiply(A=@n22.G, B=@n19.G))
  DepthFade(Opacity=Multiply(A=ParticleColor().A, B=Saturate(None=@n23)), FadeDistance=ScalarParameter() {defaultValue=20, parameterName=DepthDist})
  n24 = VectorParameter() {parameterName=Color B, defaultValue=<rgba 0 20 4 0>}
  n25 = VectorParameter() {parameterName=Color A, defaultValue=<rgba 0.024 0.3 0.6 0>}
  MaterialFunctionCall(Emissive=Add(A=Multiply(A=@n22.G, B=@n24.RGB), B=@n25.RGB)) {materialFunction=&MF_FX_EmissiveAdjustments}
  n26 = Sine(None=Multiply(A=Time()) {constB=0.1})
  n27 = AppendVector(A=Constant(), B=AppendVector(A=Multiply(A=@n26) {constB=0.1}, B=Multiply(A=@n26) {constB=-0.05}))
  n28 = Add(A=@n25.RGB, B=@n27)
  n29 = Add(A=@n27, B=@n24.RGB)
  n30 = ScalarParameter() {defaultValue=0.75, parameterName=LERP Color Exponent}
  n31 = ScalarParameter() {parameterName=PanSpeedU}
  n32 = ScalarParameter() {parameterName=PanSpeedV}
  n33 = TextureSampleParameter2D(UVs=Panner(Speed=AppendVector(A=@n31, B=@n32))) {parameterName=Mask, texture=/Game/Effects/Fort_Effects/Textures/Smoke/T_Lobber_Trail_Mask}
  n34 = Multiply(
   A=TextureCoordinate()
   B=AppendVector(A=ScalarParameter() {defaultValue=1, parameterName=Noise BG U}, B=ScalarParameter() {defaultValue=0.04, parameterName=Noise BG V})
  )
  n35 = Multiply(A=Time(), B=ScalarParameter() {defaultValue=0.15, parameterName=Speed BG})
  n36 = Add(
   A=TextureSample(UVs=Panner(Coordinate=@n34, Time=@n35) {speedX=-1, speedY=0.066}) {texture=&T_NoisePack_01}.R
   B=ScalarParameter() {defaultValue=0.25, parameterName=BG Bias}
  )
  n37 = ScalarParameter() {defaultValue=1.7, parameterName=BG Scale}
  n38 = Multiply(A=Constant() {r=1}, B=Add(A=@n9, B=Multiply(A=@n36, B=@n37)))
  n39 = Multiply(A=Add(A=Power(Base=@n9, Exp=@n30), B=@n33.G), B=@n38)
  n40 = ParticleColor()
  Multiply(
   A=Multiply(A=LinearInterpolate(A=@n28, B=@n29, Alpha=@n39), B=ScalarParameter() {defaultValue=0.25, parameterName=Emissive Boost})
   B=@n40.RGB
  )
  n41 = Add(A=@n33.G) {constB=0.2}
  n42 = ComponentMask(None=TextureCoordinate()) {r=true}
  n43 = Multiply(A=@n42, B=OneMinus(None=@n42))
  n44 = ScalarParameter() {defaultValue=4, parameterName=Gradient Multiplier}
  n45 = Multiply(A=@n38, B=Clamp(None=Multiply(A=@n43, B=@n44)))
  n46 = ComponentMask(None=TextureCoordinate()) {g=true}
  n47 = Multiply(A=@n46, B=OneMinus(None=@n46))
  n48 = Clamp(None=Multiply(A=Multiply(A=@n47, B=@n47)) {constB=16})
  n49 = Multiply(
   A=MaterialFunctionCall(Fade Length=ScalarParameter() {defaultValue=75, parameterName=CamFade}) {materialFunction=&CameraDepthFade}.Result
   B=Multiply(A=Clamp(None=@n41), B=Multiply(A=@n45, B=@n48))
  )
  n50 = ScalarParameter() {defaultValue=250, parameterName=Fade Distance}
  Multiply(A=@n40.A, B=Multiply(A=DepthFade(Opacity=@n49, FadeDistance=@n50), B=ScalarParameter() {defaultValue=5, parameterName=Opacity}))
 ]
}
