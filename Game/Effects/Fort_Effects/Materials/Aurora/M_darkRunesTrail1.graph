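# M_darkRunesTrail1 is an unlit translucent effect material for beam trails and Niagara ribbons, in the Aurora effects folder.
# An inverted U gradient and a parameterized square edge mask shape the alpha, fading it along the trail and at its edges.
# Particle alpha multiplies that mask, followed by camera-distance fade and depth fade to soften intersections.
# Particle color passes through a shared emissive adjustment function, with no textures sampled in the graph.
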Material /Game/Effects/Fort_Effects/Materials/Aurora/M_darkRunesTrail1
graph {
 properties {
  bUsedWithBeamTrails=true
  bUsedWithNiagaraRibbons=true
  blendMode=BLEND_Translucent
  pixelDepthOffsetMode=PDOM_Legacy
  shadingModel=MSM_Unlit
  usedShadingModels=Unlit
 }
 nodes [
  n1 = ParticleColor()
  n2 = MaterialFunctionCall(UV Channel=TextureCoordinate()) {materialFunction=&LinearGradient}
  n3 = Multiply(
   A=Clamp(None=OneMinus(None=@n2.UGradient))
   B=MaterialFunctionCall!(UVs, Mult=ScalarParameter() {defaultValue=4, parameterName=EdgeMult}, Exp=ScalarParameter() {defaultValue=2, parameterName=EdgeExp}) {materialFunction=/Game/Effects/Fort_Effects/MaterialFunctions/MF_SquareEdgeMask}.SquareMask
  )
  n4 = ScalarParameter() {defaultValue=75, parameterName=CamFade}
  DepthFade(
   Opacity=Multiply(
    A=Multiply(A=@n1.A, B=@n3)
    B=MaterialFunctionCall(Fade Length=@n4) {materialFunction=&CameraDepthFade}.Result
   )
   FadeDistance=ScalarParameter() {defaultValue=20, parameterName=DepthFadeDist}
  )
  MaterialFunctionCall(Emissive=@n1.RGB) {materialFunction=&MF_FX_EmissiveAdjustments}
 ]
}
